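# Post-process style Anim Blueprint for the FloralMane Sand head, adding secondary whisker motion to the input pose.
# Six chained root-relative AnimDynamics nodes apply spring-limited, damped physics to left and right whisker bones, with dyn_whisker_A_l_01 simulated twice and no C_l node.
# Skydiving, grind rail and zipline events set AnimDynAlpha, though node alphas show a constant 1.
# Zipline exit sets it to 0, like entry, instead of restoring 1.
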AnimBlueprint /BRCosmetics/Characters/Player/Male/Medium/Heads/M_MED_FloralMane_Sand_Head/Meshes/Fortnite_Base_Head_FloralMane_Sand_Post_AnimBP
target_skeleton=&/Fortnite_Base_Head_FloralMane_Sand_Skeleton
graphs [
 AnimGraph {nodes [
   n1 = Anim Dynamics - Bone: dyn_whisker_A_l_01(ComponentPose=Local To Component(LocalPose=Input Pose()), Alpha=1, ExternalForce="0,0,0") {
    angularDampingOverride=0.999
    angularSpringConstant=15000
    bAngularSpring=true
    bOverrideAngularDamping=true
    bUseGravityOverride=true
    simulationSpace=RootRelative
    alphaScaleBiasClamp {bInterpResult=true}
    boundBone {boneName=dyn_whisker_A_l_01}
    componentAppliedLinearAccClamp=<3000 3000 3000>
    gravityOverride {z=-980}
    physicsBodyDefinitions [{
      boundBone {boneName=dyn_whisker_A_l_01}
      boxExtents=<35 35 35>
      constraintSetup {angularLimitsMax=<0 10 10>, angularLimitsMin=<0 -10 -10>, angularTarget=<1 0 0>}
      localJointOffset=<1 0 0>
     }]
    simSpaceSettings {maxAngularAcceleration=2000, maxAngularVelocity=40, simSpaceAngularAlpha=0.4}
   }
   n2 = Anim Dynamics - Bone: dyn_whisker_B_l_01(ComponentPose=@n1, Alpha=1, ExternalForce="0,0,0") {
    angularDampingOverride=0.999
    angularSpringConstant=15000
    bAngularSpring=true
    bOverrideAngularDamping=true
    bUseGravityOverride=true
    simulationSpace=RootRelative
    alphaScaleBiasClamp {bInterpResult=true}
    boundBone {boneName=dyn_whisker_B_l_01}
    componentAppliedLinearAccClamp=<3000 3000 3000>
    gravityOverride {z=-980}
    physicsBodyDefinitions [{
      boundBone {boneName=dyn_whisker_B_l_01}
      boxExtents=<35 35 35>
      constraintSetup {angularLimitsMax=<0 10 10>, angularLimitsMin=<0 -10 -10>, angularTarget=<1 0 0>}
      localJointOffset=<1.25 0 0>
     }]
    simSpaceSettings {maxAngularAcceleration=2000, maxAngularVelocity=40, simSpaceAngularAlpha=0.4}
   }
   n3 = Anim Dynamics - Bone: dyn_whisker_A_l_01(ComponentPose=@n2, Alpha=1, ExternalForce="0,0,0") {
    angularDampingOverride=0.999
    angularSpringConstant=15000
    bAngularSpring=true
    bOverrideAngularDamping=true
    bUseGravityOverride=true
    simulationSpace=RootRelative
    alphaScaleBiasClamp {bInterpResult=true}
    boundBone {boneName=dyn_whisker_A_l_01}
    componentAppliedLinearAccClamp=<3000 3000 3000>
    gravityOverride {z=-980}
    physicsBodyDefinitions [{
      boundBone {boneName=dyn_whisker_A_l_01}
      boxExtents=<35 35 35>
      constraintSetup {angularLimitsMax=<0 10 10>, angularLimitsMin=<0 -10 -10>, angularTarget=<1 0 0>}
      localJointOffset=<1.4 0 0>
     }]
    simSpaceSettings {maxAngularAcceleration=2000, maxAngularVelocity=40, simSpaceAngularAlpha=0.4}
   }
   n4 = Anim Dynamics - Bone: dyn_whisker_A_r_01(ComponentPose=@n3, Alpha=1, ExternalForce="0,0,0") {
    angularDampingOverride=0.999
    angularSpringConstant=15000
    bAngularSpring=true
    bOverrideAngularDamping=true
    bUseGravityOverride=true
    simulationSpace=RootRelative
    alphaScaleBiasClamp {bInterpResult=true}
    boundBone {boneName=dyn_whisker_A_r_01}
    componentAppliedLinearAccClamp=<3000 3000 3000>
    gravityOverride {z=-980}
    physicsBodyDefinitions [{
      boundBone {boneName=dyn_whisker_A_r_01}
      boxExtents=<35 35 35>
      constraintSetup {angularLimitsMax=<0 10 10>, angularLimitsMin=<0 -10 -10>, angularTarget=<1 0 0>}
      localJointOffset=<1.25 0 0>
     }]
    simSpaceSettings {maxAngularAcceleration=2000, maxAngularVelocity=40, simSpaceAngularAlpha=0.4}
   }
   n5 = Anim Dynamics - Bone: dyn_whisker_B_r_01(ComponentPose=@n4, Alpha=1, ExternalForce="0,0,0") {
    angularDampingOverride=0.999
    angularSpringConstant=15000
    bAngularSpring=true
    bOverrideAngularDamping=true
    bUseGravityOverride=true
    simulationSpace=RootRelative
    alphaScaleBiasClamp {bInterpResult=true}
    boundBone {boneName=dyn_whisker_B_r_01}
    componentAppliedLinearAccClamp=<3000 3000 3000>
    gravityOverride {z=-980}
    physicsBodyDefinitions [{
      boundBone {boneName=dyn_whisker_B_r_01}
      boxExtents=<35 35 35>
      constraintSetup {angularLimitsMax=<0 10 10>, angularLimitsMin=<0 -10 -10>, angularTarget=<1 0 0>}
      localJointOffset=<1.4 0 0>
     }]
    simSpaceSettings {maxAngularAcceleration=2000, maxAngularVelocity=40, simSpaceAngularAlpha=0.4}
   }
   n6 = Anim Dynamics - Bone: dyn_whisker_C_r_01(ComponentPose=@n5, Alpha=1, ExternalForce="0,0,0") {
    angularDampingOverride=0.999
    angularSpringConstant=15000
    bAngularSpring=true
    bOverrideAngularDamping=true
    bUseGravityOverride=true
    simulationSpace=RootRelative
    alphaScaleBiasClamp {bInterpResult=true}
    boundBone {boneName=dyn_whisker_C_r_01}
    componentAppliedLinearAccClamp=<3000 3000 3000>
    gravityOverride {z=-980}
    physicsBodyDefinitions [{
      boundBone {boneName=dyn_whisker_C_r_01}
      boxExtents=<35 35 35>
      constraintSetup {angularLimitsMax=<0 10 10>, angularLimitsMin=<0 -10 -10>, angularTarget=<1 0 0>}
      localJointOffset=<1 0 0>
     }]
    simSpaceSettings {maxAngularAcceleration=2000, maxAngularVelocity=40, simSpaceAngularAlpha=0.4}
   }
   Output Pose(Result=Component To Local(ComponentPose=@n6))
  ]}
 EventGraph {nodes [
   Event OnBeginSkydiving()
   > Set AnimDynAlpha(AnimDynAlpha:double=0) -> (Output_Get:double)
   Event OnEndSkydiving()
   > Set AnimDynAlpha(AnimDynAlpha=1)
   Event OnEnterGrindRail() {is=title K2Node_Event}
   > Set AnimDynAlpha(AnimDynAlpha=0)
   n1 = Event BlueprintProcessAnimInputEvent()
   > n2 = Branch(Condition=IsNamed(self=@n1.AnimInputEvent, InEventName=aie.Player.Zipline.Entered))
   > Set AnimDynAlpha(AnimDynAlpha=0)
   Branch(execute=@n2.else, Condition=IsNamed(self=@n1.AnimInputEvent, InEventName=aie.Player.Zipline.Exited))
   > Set AnimDynAlpha(AnimDynAlpha=0)
   Event OnExitGrindRail() {is=title K2Node_Event}
   > Set AnimDynAlpha(AnimDynAlpha=1)
   TryGetPawnOwner()
   Event BlueprintUpdateAnimation()
  ]}
]
variables [AnimDynAlpha {type=number, default=1}]
